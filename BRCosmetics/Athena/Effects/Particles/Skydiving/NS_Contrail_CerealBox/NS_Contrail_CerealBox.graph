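# Cereal-themed skydiving contrail cosmetic that trails colored mesh pieces, droplets and ribbons from the character's skeleton.
# Cereal meshes spawn in four color groups and swirl on looping vector curves, with spawn rate and motion scaled by dive amount.
# Two near-identical hand ribbon emitters stretch with owner speed, alongside persistent hand-attached meshes and GPU droplet sprites.
# Front-end modifiers retune rates and velocities, and time of day scales ribbon alpha.
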NiagaraSystem /BRCosmetics/Athena/Effects/Particles/Skydiving/NS_Contrail_CerealBox/NS_Contrail_CerealBox
dependencies {
 usedDataInterfaces [
  &NiagaraDataInterfaceArrayColor
  &NiagaraDataInterfaceVectorCurve
  &NiagaraDataInterfaceCurve
  &NiagaraDataInterfaceColorCurve
 ]
 usedDynamicInputs []
 usedModules [&SystemState, /Game/Effects/Niagara/Modules/Skydiving/FN_DivingAttributes/v4/FN_DivingAttributes_v4]
}
emitters {
 Cereal01001 {
  emitterUpdateScript [
   &EmitterState()
   &SpawnRate(
    SpawnRate=&Lerp_Float(A=6, B=&FN_FrontEndFloatModifier(Gameplay=10, Front End=29), Alpha=$System.Dive)
    Use Spawn Probability=false
   )
   &SpawnRate(
    SpawnGroup=1
    SpawnRate=&Lerp_Float(A=6, B=&FN_FrontEndFloatModifier(Gameplay=7, Front End=40), Alpha=$System.Dive)
    Use Spawn Probability=false
   ) {module=SpawnRate002}
   &SpawnRate(
    SpawnGroup=2
    SpawnRate=&Lerp_Float(A=4, B=&FN_FrontEndFloatModifier(Gameplay=9, Front End=30), Alpha=$System.Dive)
    Use Spawn Probability=false
   ) {module=SpawnRate001}
   &SpawnRate(
    SpawnGroup=3
    SpawnRate=&Lerp_Float(A=5, B=&FN_FrontEndFloatModifier(Gameplay=9, Front End=28), Alpha=$System.Dive)
    Use Spawn Probability=false
   ) {module=SpawnRate003}
  ]
  particleSpawnScript [
   &InitializeParticle(
    Color=&SelectLinearColorFromArray(
     Array Sampling Mode=::NewEnumerator1
     Direct Array Index=$Emitter.SpawnGroup
     Color Selection Array=DataInterface() {propertyValues {ColorData [<rgba 0.0829 0.6911 0.9948 1>, <rgba 1 0.4051 0 1>, <rgba 1 0.0489 0.5056 1>, <rgba 0.5469 1 0.0938 1>]}}
     Random Seed=?
    )
    Lifetime Max=1
    Lifetime Min=0.6
    Lifetime Mode=::NewEnumerator1
    Mesh Scale Mode=::NewEnumerator3
    Mesh Uniform Scale=&RandomRangeFloat(Minimum=1.2, Maximum=1.45)
    UsePositionOffset=false
    Write Material Random=true
   )
   &SkeletalMeshLocation(
    Bone / Socket Index=&RandomRangeInt(Minimum=0, Maximum=1)
    Bone Sampling Mode=::NewEnumerator4
    MaskSpawnGroup=false
    Skeletal Mesh=$User.SkeletalMesh
   )
   &AddVelocity(Velocity=&RandomRangeVector(Minimum=<0 0 0>, Maximum=<1 1 0>))
   &InitialMeshOrientation(
    Rotation=&RandomRangeVector(Minimum=<0 0 0>, Maximum=<1 1 1>)
    Use Orientation Vector=true
    Use Rotation Vector=true
   )
   &FN_SetSpawnGroup()
  ]
  particleUpdateScript [
   /Game/Effects/Niagara/Modules/Skydiving/FN_DivingAttributes/FN_DivingAttributes_v2(
    Accumulate Dive=1
    Accumulate Rest=0
    Bone Index=0
    SkeletalMesh=DataInterface() {propertyValues {
      DefaultMesh=null
      ExcludeBoneName=root
      OptionalComponentClass=null
      PreviewMesh=&M_Med_Soldier_04_Preview
      SkinningMode=SkinOnTheFly
      SoftSourceActor=null
      SourceMode=Default
      UvSetIndex=0
      WholeMeshLOD=-1
      bExcludeBone=false
      bReadDeformedGeometry=true
      bRequireCurrentFrameData=true
      ComponentTags []
      FilteredBones [spine_02]
      FilteredSockets []
      MeshUserParameter {parameter {name=None, typeDefHandle {registeredTypeIndex=1}, varData []}}
      SamplingRegions []
     }}
   )
   /SetVariables(Particles.CustomizePosition:Vector3f=&Multiply_Vector(
     A=&Lerp_Vector(
      A=&FN_FrontEndVector3Modifier(Gameplay=<180 180 200>, Front End=<180 180 120>)
      B=&FN_FrontEndVector3Modifier(Gameplay=<180 180 400>, Front End=<180 180 700>)
      Alpha=$Particles.Dive
     )
     B=&VectorFromCurve(VectorCurve=DataInterface() {propertyValues {
        XCurve {keys [<at 0 1>,<at 0.12 0>,<at 0.24 -1>,<at 0.36 0>,<at 0.48 1>,<at 0.6 0>,<at 0.72 -1>,<at 0.84 0>,<at 0.96 1>,<at 1.08 0>,<at 1.2 -1>,<at 1.32 0>,<at 1.44 1>,<at 1.56 0>]}
        YCurve {keys [<at 0 0>,<at 0.12 1>,<at 0.24 0>,<at 0.36 -1>,<at 0.48 0>,<at 0.6 1>,<at 0.72 0>,<at 0.84 -1>,<at 0.96 0>,<at 1.08 1>,<at 1.2 0>,<at 1.32 -1>,<at 1.44 0>,<at 1.56 1>]}
        ZCurve {keys [<at 0 1.0067>]}
       }})
    )) {isSetParameters=true}
   /SetVariables(Particles.Velocity=$Particles.CustomizePosition) {isSetParameters=true}
   &InheritVelocity(Inherited Velocity Amount Scale=<1 1 0>, Inherited Velocity Speed Limit=0, Speed Limit=true, UseSpeedThreshold=false)
   &SolveForcesAndVelocity(Clamp Velocity=false, Limit Acceleration=false)
   &ParticleState()
   &UpdateMeshOrientation(Rotation Rate=&RandomRangeFloat(Minimum=-1, Maximum=1), Rotation Vector=<1 1 1>)
   &ScaleMeshSize(Scale Factor=&Multiply_VectorByFloat(
     Vector=<1 1 1>
     Float=&FloatFromCurve(FloatCurve=DataInterface() {propertyValues {Curve {keys [<at 0 0>,<at 0.2 1>,<at 0.75 1>,<at 1 0>]}}})
    ))
  ]
  rendererClasses [&NiagaraMeshRendererProperties]
  renderers [{rendererClass=&NiagaraMeshRendererProperties, rendererIndex=0}]
 }
 Drops {
  simTarget=GPUComputeSim
  emitterUpdateScript [
   &EmitterState(Inactive Response=::NewEnumerator1, Life Cycle Mode=::NewEnumerator1, Loop Behavior=::NewEnumerator1, Loop Duration=3600, UseLoopDelay=false)
   &SpawnRate(
    SpawnGroup:NiagaraInt32=&RandomRangeInt(Minimum=0, Maximum=1)
    SpawnRate=50
    Use Spawn Probability=false
   )
  ]
  particleSpawnScript [
   &FN_SetSpawnGroup(Number of Spawn Groups=2)
   &InitializeParticle(
    Lifetime Max=0.5
    Lifetime Min=0.5
    Lifetime Mode=::NewEnumerator1
    Sprite Size Mode=::NewEnumerator1
    Uniform Sprite Size Min=10
    UsePositionOffset=false
    Write Material Random=true
   ) {module=InitializeParticle001}
   &SkeletalMeshLocation(
    Bone / Socket Index=$Particles.SpawnGroup
    Bone Sampling Mode=::NewEnumerator4
    Inherit Velocity=false
    Interpolate Mesh Sampling=false
    MaskSpawnGroup=false
    Sampled Position Offset Coordinate Space:ENiagaraMeshTransforms=::NewEnumerator3
    Skeletal Mesh=$User.SkeletalMesh
   )
   /SetVariables(Particles.Velocity=&UniformRangedVector(Minimum=<-100 -100 200>, Maximum=<100 100 300>)) {isSetParameters=true}
   /SetVariables(Particles.SpriteSize=&Multiply_Vector2(
     A=<5 10>
     B=&Vector2DFromFloat(Value=&UniformRangedFloat(Minimum=1, Maximum=2))
    )) {isSetParameters=true}
  ]
  particleUpdateScript [
   &ParticleState()
   &GravityForce(Gravity=<0 0 -100>)
   &Drag(Use Linear Drag=true, Use Rotational Drag=false)
   &SolveForcesAndVelocity(Clamp Velocity=false, Limit Acceleration=false)
   &ScaleColor(
    Scale Alpha=&FloatFromCurve(FloatCurve=DataInterface() {propertyValues {Curve {keys [
         <at 0 0 RCIM_Cubic>
         {arriveTangent=1.6041, leaveTangent=1.6041, time=0.05, value=1.0113}
         {interpMode=RCIM_Cubic, leaveTangent=-1.0633, time=0.75, value=1}
         <at 0.99 0 RCIM_Cubic>
        ]}}})
    Scale RGB=<0.6 0.8 1>
    ScaleA=true
    ScaleRGB=true
   )
   &ScaleSpriteSize(Scale Factor=&Vector2DFromFloat(Value=&FloatFromCurve(
      FloatCurve=DataInterface() {propertyValues {Curve {keys [
          {arriveTangent=-0.4919, interpMode=RCIM_Cubic, leaveTangent=-0.4919, tangentMode=RCTM_User, time=0.25, value=2}
          {arriveTangent=-3.3301, interpMode=RCIM_Cubic, leaveTangent=-3.3302, tangentMode=RCTM_User, time=1, value=0}
         ]}}}
      Scale Curve=0.65
     )))
  ]
  rendererClasses [&NiagaraSpriteRendererProperties]
  renderers [{rendererClass=&NiagaraSpriteRendererProperties}]
 }
 Mike_Hand {
  emitterUpdateScript [
   &EmitterState(
    Inactive Response=::NewEnumerator1
    Life Cycle Mode=::NewEnumerator1
    Loop Behavior=::NewEnumerator1
    Loop Duration Mode=::NewEnumerator1
    UseLoopDelay=false
   )
   &SpawnBurst_Instantaneous(Spawn Count=2, Use Spawn Probability=false, UseLoopCountLimit=false) {module=SpawnBurst_Instantaneous002}
  ]
  particleSpawnScript [
   &FN_SetSpawnGroup(Number of Spawn Groups=2)
   /Niagara/Modules/Spawn/Initialization/InitializeParticle(
    Color=<rgba 1 1 1 0.47>
    Mesh Scale=&Multiply_VectorByFloat(Vector=<1 1 1>, Float=&FN_FrontEndFloatModifier(Gameplay=0.9, Front End=0.9))
    Module.Write Color=true
    Module.Write Lifetime=true
    Module.Write Mass=true
    Module.Write Material Random=true
    Module.Write Position=false
    Module.Write Scale=true
    Module.Write SpriteRotation=false
    Module.Write SpriteSize=true
    Position=&SimulationPosition()
    Sprite Size=<5 5>
   )
  ]
  particleUpdateScript [
   &SkeletalMeshLocation(
    Bone / Socket Index=$Particles.SpawnGroup
    Bone Sampling Mode=::NewEnumerator4
    Inherit Velocity=false
    MaskSpawnGroup=false
    Skeletal Mesh=DataInterface() {propertyValues {
      DefaultMesh=null
      ExcludeBoneName=root
      OptionalComponentClass=null
      PreviewMesh=&M_Med_Soldier_04_Preview
      SkinningMode=SkinOnTheFly
      SoftSourceActor=null
      SourceMode=Default
      UvSetIndex=0
      WholeMeshLOD=-1
      bExcludeBone=false
      bReadDeformedGeometry=true
      bRequireCurrentFrameData=true
      ComponentTags []
      FilteredBones [hand_l,hand_r]
      FilteredSockets []
      MeshUserParameter {parameter {name=None, typeDefHandle {registeredTypeIndex=1}, varData []}}
      SamplingRegions []
     }}
   ) {module=SkeletalMeshLocation001}
   &ParticleState(Kill Particles When Lifetime Has Elapsed=false, Let Infinitely Lived Particles Die When Emitter Deactivates=true)
   &SolveForcesAndVelocity(Clamp Velocity=false, Limit Acceleration=false)
   &DynamicMaterialParameters(
    Index 0 Param 1=&RandomRangeFloat(Minimum=0, Maximum=1)
    Index 0 Param 2=5
    Index 0 Param 3=1
    Index 0 Param 4=1
    Module.Param0WriteEnabled=true
    Module.Param1WriteEnabled=true
    Module.Param2WriteEnabled=true
    Module.Param3WriteEnabled=true
   )
  ]
  rendererClasses [&NiagaraMeshRendererProperties]
  renderers [{rendererClass=&NiagaraMeshRendererProperties, rendererIndex=0}]
 }
 Paint_Ribbon01 {
  emitterUpdateScript [
   &EmitterState()
   /Game/Effects/Niagara/Modules/Ribbons/FN_SpawnRibbon_v3(
    Number of Ribbons=2
    Skeleton Mesh=DataInterface() {propertyValues {
      DefaultMesh=null
      ExcludeBoneName=root
      OptionalComponentClass=null
      PreviewMesh=&M_Med_Soldier_04_Preview
      SkinningMode=SkinOnTheFly
      SoftSourceActor=null
      SourceMode=Default
      UvSetIndex=0
      WholeMeshLOD=-1
      bExcludeBone=false
      bReadDeformedGeometry=true
      bRequireCurrentFrameData=true
      ComponentTags []
      FilteredBones [hand_l,hand_r]
      FilteredSockets []
      MeshUserParameter {parameter {name=None, typeDefHandle {registeredTypeIndex=1}, varData []}}
      SamplingRegions []
     }}
   )
  ]
  particleSpawnScript [
   &FN_SetSpawnGroup(Number of Spawn Groups=2)
   /Game/Effects/Niagara/Modules/Ribbons/FN_InitializeRibbon_v5(Enable Skeletal Position=true, Lifetime=1, Ribbon Width=65, SkeletalMesh:NiagaraDataInterfaceSkeletalMesh=$User.SkeletalMesh)
   &AddVelocity(Velocity=&FN_FrontEndVector3Modifier(Gameplay=<0 0 300>, Front End=<0 0 400>)) {module=AddVelocity001}
   /SetVariables(Particles.Color=&LerpLinearColors(
     StartColor=&ColorFromCurve(
      ColorCurve=DataInterface() {propertyValues {
        AlphaCurve {keys [<at 0 1 RCIM_Cubic>]}
        BlueCurve {keys [<at 0 1 RCIM_Cubic>,<at 1 1 RCIM_Cubic>]}
        GreenCurve {keys [<at 0 1 RCIM_Cubic>,<at 1 1 RCIM_Cubic>]}
        RedCurve {keys [<at 0 1 RCIM_Cubic>,<at 1 1 RCIM_Cubic>]}
       }}
      CurveIndex=&RandomRangeFloat(Minimum=0, Maximum=1)
     )
     EndColor=&ColorFromCurve(
      ColorCurve=DataInterface() {propertyValues {
        AlphaCurve {keys [<at 0 1 RCIM_Cubic>]}
        BlueCurve {keys [<at 0 1 RCIM_Cubic>,<at 1 1 RCIM_Cubic>]}
        GreenCurve {keys [<at 0 1 RCIM_Cubic>,<at 1 1 RCIM_Cubic>]}
        RedCurve {keys [<at 0 1 RCIM_Cubic>,<at 1 1 RCIM_Cubic>]}
       }}
      CurveIndex=&RandomRangeFloat(Minimum=0, Maximum=1)
     )
     LerpFactor=&MakeFloatFromInt(Int=$Particles.SpawnGroup)
    )) {isSetParameters=true, module=SetVariables_#1}
   /SetVariables(Particles.MaterialRandom:NiagaraFloat=&RandomRangeFloat(Minimum=0, Maximum=1)) {isSetParameters=true, module=SetVariables_#2}
   /SetVariables() {isSetParameters=true, module=SetVariables_#3}
  ]
  particleUpdateScript [
   &ScaleVelocity(Velocity Scale=&MakeVector(
     X=0
     Y=0
     Z=&Multiply_Float(
      A=0.9
      B=&FloatFromCurve(
       FloatCurve=DataInterface() {propertyValues {Curve {keys [<at 0 0.9>,<at 1 3>]}}}
       CurveIndex=&Divide_Float(A=&VectorLength(Vector=$Engine.Owner.Velocity), B=9000)
      )
     )
    )) {module=ScaleVelocity001}
   &SolveForcesAndVelocity(Clamp Velocity=false, Limit Acceleration=false)
   /Game/Effects/Niagara/Modules/Ribbons/FN_UpdateRibbon_v6(Enable Skeletal Position=true, SkeletalMesh:NiagaraDataInterfaceSkeletalMesh=$User.SkeletalMesh)
   &DynamicMaterialParameters(
    Index 0 Param 1=&MaskFloatBySpawnGroup(Masked Float=0, Passthrough Float=0.5)
    Index 0 Param 2=&FloatFromCurve(FloatCurve=DataInterface() {propertyValues {Curve {keys [
         {arriveTangent=-0.3062, interpMode=RCIM_Cubic, leaveTangent=-0.3062, tangentMode=RCTM_User, time=0, value=0.5}
         {arriveTangent=0.0167, interpMode=RCIM_Cubic, leaveTangent=0.0167, tangentMode=RCTM_Break, time=1, value=0}
        ]}}})
    Index 0 Param 3=1
    Index 0 Param 4=1
    Module.Param0WriteEnabled=true
    Module.Param1WriteEnabled=true
    Module.Param2WriteEnabled=true
    Module.Param3WriteEnabled=true
   )
   &ScaleRibbonWidth(Ribbon Width Scale=&FloatFromCurve(
     FloatCurve=DataInterface() {propertyValues {Curve {keys [
         {arriveTangent=-0.0063, interpMode=RCIM_Cubic, leaveTangent=-0.0063, tangentMode=RCTM_User, time=0, value=1}
         {arriveTangent=2.3571, interpMode=RCIM_Cubic, leaveTangent=2.3571, time=0.4, value=2.025}
         <at 0.7 2.325>
        ]}}}
     Scale Curve=0.6
    ))
   &ScaleColor(
    Scale Alpha=&Multiply_Float(
     A=&FNTimeOfDayScalar(Night Scale=1.5, Morning Scale=2, Evening Scale=2)
     B=&FloatFromCurve(FloatCurve=DataInterface() {propertyValues {Curve {keys [<at 0 1>,<at 0.1 1>,<at 1 1>]}}})
    )
    Scale RGB=&Multiply_Vector(
     A=&MakeVectorFromLinearColorRGB(Color=<rgba 1 1 1 1>)
     B=&VectorFromFloat(Value=1.5)
    )
    ScaleA=true
    ScaleRGB=true
   )
  ]
  rendererClasses [&NiagaraRibbonRendererProperties]
  renderers [{rendererClass=&NiagaraRibbonRendererProperties}]
 }
 Paint_Ribbon01001 {
  emitterUpdateScript [
   &EmitterState()
   &FN_SpawnRibbon_v3(
    Number of Ribbons=2
    Skeleton Mesh=DataInterface() {propertyValues {
      DefaultMesh=null
      ExcludeBoneName=root
      OptionalComponentClass=null
      PreviewMesh=&M_Med_Soldier_04_Preview
      SkinningMode=SkinOnTheFly
      SoftSourceActor=null
      SourceMode=Default
      UvSetIndex=0
      WholeMeshLOD=-1
      bExcludeBone=false
      bReadDeformedGeometry=true
      bRequireCurrentFrameData=true
      ComponentTags []
      FilteredBones [hand_l,hand_r]
      FilteredSockets []
      MeshUserParameter {parameter {name=None, typeDefHandle {registeredTypeIndex=1}, varData []}}
      SamplingRegions []
     }}
   )
  ]
  particleSpawnScript [
   &FN_SetSpawnGroup(Number of Spawn Groups=2)
   &FN_InitializeRibbon_v5(Enable Skeletal Position=true, Lifetime=1, Ribbon Width=65, SkeletalMesh:NiagaraDataInterfaceSkeletalMesh=$User.SkeletalMesh)
   &AddVelocity(Velocity=&FN_FrontEndVector3Modifier(Gameplay=<0 0 300>, Front End=<0 0 400>)) {module=AddVelocity001}
   /SetVariables(Particles.Color=&LerpLinearColors(
     StartColor=&ColorFromCurve(
      ColorCurve=DataInterface() {propertyValues {
        AlphaCurve {keys [<at 0 1 RCIM_Cubic>]}
        BlueCurve {keys [<at 0 1 RCIM_Cubic>,<at 1 1 RCIM_Cubic>]}
        GreenCurve {keys [<at 0 1 RCIM_Cubic>,<at 1 1 RCIM_Cubic>]}
        RedCurve {keys [<at 0 1 RCIM_Cubic>,<at 1 1 RCIM_Cubic>]}
       }}
      CurveIndex=&RandomRangeFloat(Minimum=0, Maximum=1)
     )
     EndColor=&ColorFromCurve(
      ColorCurve=DataInterface() {propertyValues {
        AlphaCurve {keys [<at 0 1 RCIM_Cubic>]}
        BlueCurve {keys [<at 0 1 RCIM_Cubic>,<at 1 1 RCIM_Cubic>]}
        GreenCurve {keys [<at 0 1 RCIM_Cubic>,<at 1 1 RCIM_Cubic>]}
        RedCurve {keys [<at 0 1 RCIM_Cubic>,<at 1 1 RCIM_Cubic>]}
       }}
      CurveIndex=&RandomRangeFloat(Minimum=0, Maximum=1)
     )
     LerpFactor=&MakeFloatFromInt(Int=$Particles.SpawnGroup)
    )) {isSetParameters=true, module=SetVariables_#1}
   /SetVariables(Particles.MaterialRandom:NiagaraFloat=&RandomRangeFloat(Minimum=0, Maximum=1)) {isSetParameters=true, module=SetVariables_#2}
   /SetVariables() {isSetParameters=true, module=SetVariables_#3}
  ]
  particleUpdateScript [
   &ScaleVelocity(Velocity Scale=&MakeVector(
     X=0
     Y=0
     Z=&Multiply_Float(
      A=0.9
      B=&FloatFromCurve(
       FloatCurve=DataInterface() {propertyValues {Curve {keys [<at 0 0.9>,<at 1 3>]}}}
       CurveIndex=&Divide_Float(A=&VectorLength(Vector=$Engine.Owner.Velocity), B=9000)
      )
     )
    )) {module=ScaleVelocity001}
   &SolveForcesAndVelocity(Clamp Velocity=false, Limit Acceleration=false)
   &FN_UpdateRibbon_v6(Enable Skeletal Position=true, SkeletalMesh:NiagaraDataInterfaceSkeletalMesh=$User.SkeletalMesh)
   &DynamicMaterialParameters(
    Index 0 Param 1=&MaskFloatBySpawnGroup(Masked Float=0, Passthrough Float=0.5)
    Index 0 Param 2=&FloatFromCurve(FloatCurve=DataInterface() {propertyValues {Curve {keys [
         {arriveTangent=-0.3062, interpMode=RCIM_Cubic, leaveTangent=-0.3062, tangentMode=RCTM_User, time=0, value=0.5}
         {arriveTangent=0.0167, interpMode=RCIM_Cubic, leaveTangent=0.0167, tangentMode=RCTM_Break, time=1, value=0}
        ]}}})
    Index 0 Param 3=1
    Index 0 Param 4=1
    Module.Param0WriteEnabled=true
    Module.Param1WriteEnabled=true
    Module.Param2WriteEnabled=true
    Module.Param3WriteEnabled=true
   )
   &ScaleRibbonWidth(Ribbon Width Scale=&FloatFromCurve(
     FloatCurve=DataInterface() {propertyValues {Curve {keys [
         {arriveTangent=-0.0063, interpMode=RCIM_Cubic, leaveTangent=-0.0063, tangentMode=RCTM_User, time=0, value=1}
         {arriveTangent=2.3571, interpMode=RCIM_Cubic, leaveTangent=2.3571, time=0.4, value=2.025}
         <at 0.7 2.325>
        ]}}}
     Scale Curve=0.6
    ))
   &ScaleColor(
    Scale Alpha=&Multiply_Float(
     A=&FNTimeOfDayScalar(Night Scale=1.5, Morning Scale=2, Evening Scale=2)
     B=&FloatFromCurve(FloatCurve=DataInterface() {propertyValues {Curve {keys [<at 0 1>,<at 0.33 1>,<at 1 1>]}}})
    )
    Scale RGB=&Multiply_Vector(
     A=&MakeVectorFromLinearColorRGB(Color=<rgba 1 1 1 1>)
     B=&VectorFromFloat(Value=1.5)
    )
    ScaleA=true
    ScaleRGB=true
   )
  ]
  rendererClasses [&NiagaraRibbonRendererProperties]
  renderers [{rendererClass=&NiagaraRibbonRendererProperties}]
 }
}
summary {
 systemName=NS_Contrail_CerealBox
 emitters [
  {emitterName=Mike_Hand, rendererClasses [&NiagaraMeshRendererProperties]}
  {emitterName=Cereal01001, rendererClasses [&NiagaraMeshRendererProperties]}
  {emitterName=Paint_Ribbon01, rendererClasses [&NiagaraRibbonRendererProperties]}
  {emitterName=Drops, simTarget=GPUComputeSim, rendererClasses [&NiagaraSpriteRendererProperties]}
  {emitterName=Paint_Ribbon01001, rendererClasses [&NiagaraRibbonRendererProperties]}
 ]
 userVariables [
  User.bIsFrontEnd {type=NiagaraBool}
  User.bIsFrontEndPreview {type=NiagaraBool}
  User.PlayerSpeed {type=NiagaraFloat, defaultValue=0, description=""}
  User.RibbonIDOffset {type=NiagaraInt32}
  User.SkeletalMesh {
   type=NiagaraDataInterfaceSkeletalMesh
   defaultValue {
    struct=NiagaraExt_VariableValue_DataInterface
    value {dataInterface=&self:NiagaraDataInterfaceSkeletalMesh_0}
   }
  }
 ]
}
